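# Physics asset for the WartyBrine_Sus medium male body cosmetic, defining bodies for an upper and a lower dangling chain.
# Each chain runs from a small base capsule through dynamic capsule segments, with two capsules shaping the lower tip.
# Constraints link neighboring segments with twist locked and tight swing limits that are widest at the tips.
# A large, unconstrained surfaceChest capsule provides a chest collision shape.
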PhysicsAsset /BRCosmetics/Characters/Player/Male/Medium/Bodies/M_MED_WartyBrine_Sus/Meshes/M_MED_WartyBrine_Sus_Physics
bodies {
 lowerChain_Base [{shapeType=Capsule, length=1, radius=1, shapeName=lowerChain_Base_capsule, rotation=<rot 0 -90 0>}]
 lowerChain_Dyn_01 [{shapeType=Capsule, length=6, radius=2.2, shapeName=lowerChain_Dyn_01_capsule, center=<1 0 0>, rotation=<rot 90 0 0>}]
 lowerChain_Dyn_02 [{shapeType=Capsule, length=7, radius=2.2, shapeName=lowerChain_Dyn_02_capsule, center=<1 0 0>, rotation=<rot 90 0 0>}]
 lowerChain_Dyn_03 [
  {
   shapeType=Capsule
   length=17.65
   radius=1.9039
   shapeName=lowerChain_Dyn_03_capsule
   center=<7.0767 -2.7275 -5.9608>
   rotation=<rot 81.8623 96.2726 -164.7743>
  }
  {
   shapeType=Capsule
   length=16.6514
   radius=2.2073
   shapeName=lowerChain_Dyn_03_capsule
   center=<2.6425 -0.0216 -0.5296>
   rotation=<rot 36.758 -0.2261 91.738>
  }
 ]
 surfaceChest [{
   shapeType=Capsule
   length=23.5
   radius=18.5
   shapeName=surfaceChest_capsule
   center=<-6.265 -16.7369 0>
   rotation=<rot 90 -179.4032 -173.2498>
  }]
 upperChain_Base [{shapeType=Capsule, length=1, radius=1, shapeName=upperChain_Base_capsule, rotation=<rot 0 -90 0>}]
 upperChain_Dyn_01 [{shapeType=Capsule, length=1.5074, radius=2, shapeName=upperChain_Dyn_01_capsule, center=<6.3785 0 0>, rotation=<rot 90 0 0>}]
 upperChain_Dyn_02 [{shapeType=Capsule, length=14.9385, radius=2.5, shapeName=upperChain_Dyn_02_capsule, center=<1 0 0>, rotation=<rot 90 0 90>}]
}
constraints [
 {bone1Name=lowerChain_Dyn_01, bone2Name=lowerChain_Base, swing1LimitDegrees=1, swing2LimitDegrees=4, twistLimitDegrees=0, twistMotion=Locked}
 {
  bone1Name=lowerChain_Dyn_02
  bone2Name=lowerChain_Dyn_01
  swing1LimitDegrees=3
  swing2LimitDegrees=3
  twistLimitDegrees=0
  twistMotion=Locked
 }
 {
  bone1Name=lowerChain_Dyn_03
  bone2Name=lowerChain_Dyn_02
  swing1LimitDegrees=7
  swing2LimitDegrees=20
  twistLimitDegrees=0
  twistMotion=Locked
 }
 {bone1Name=upperChain_Dyn_01, bone2Name=upperChain_Base, swing1LimitDegrees=0, swing2LimitDegrees=5, twistLimitDegrees=0, twistMotion=Locked}
 {
  bone1Name=upperChain_Dyn_02
  bone2Name=upperChain_Dyn_01
  swing1LimitDegrees=4
  swing2LimitDegrees=10
  twistLimitDegrees=2
  twistMotion=Locked
 }
]
