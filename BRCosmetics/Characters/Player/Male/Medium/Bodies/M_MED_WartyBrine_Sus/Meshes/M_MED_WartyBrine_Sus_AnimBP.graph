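# Animation Blueprint for the WartyBrine_Sus cosmetic player body, layering secondary motion onto a pose copied from the attached parent mesh.
# The AnimGraph chains arm deform-bone constraints, a Control Rig, chain-bone offsets, a weighted surfaceChest constraint and RigidBody simulation.
# The EventGraph converts crouch, emote, floating, weapon and mannequin state into rig inputs.
# It also computes head, melee-arm and specific-emote values that the AnimGraph leaves unused.
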AnimBlueprint /BRCosmetics/Characters/Player/Male/Medium/Bodies/M_MED_WartyBrine_Sus/Meshes/M_MED_WartyBrine_Sus_AnimBP
target_skeleton=&/M_MED_WartyBrine_Sus_Skeleton
graphs [
 AnimGraph {nodes [
   n1 = "Constraint - deform_upperarm_l (P)"(ComponentPose=Local To Component(LocalPose=Copy Pose From Mesh() {bUseAttachedParent=true}), Alpha=1, ConstraintWeights_0=1) {
    alphaBoolBlend {blendOption=Linear}
    boneToModify {boneName=deform_upperarm_l}
    constraintSetup [{targetBone {boneName=upperarm_l}}]
    constraintWeights [1]
   }
   n2 = "Constraint - deform_lowerarm_l (P)"(ComponentPose=@n1, Alpha=1, ConstraintWeights_0=1) {
    alphaBoolBlend {blendOption=Linear}
    boneToModify {boneName=deform_lowerarm_l}
    constraintSetup [{targetBone {boneName=lowerarm_l}}]
    constraintWeights [1]
   }
   n3 = "Constraint - deform_hand_l (P)"(ComponentPose=@n2, Alpha=1, ConstraintWeights_0=1) {
    alphaBoolBlend {blendOption=Linear}
    boneToModify {boneName=deform_hand_l}
    constraintSetup [{targetBone {boneName=hand_l}}]
    constraintWeights [1]
   }
   n4 = "Constraint - deform_upperarm_r (P)"(ComponentPose=@n3, Alpha=1, ConstraintWeights_0=1) {
    alphaBoolBlend {blendOption=Linear}
    boneToModify {boneName=deform_upperarm_r}
    constraintSetup [{targetBone {boneName=upperarm_r}}]
    constraintWeights [1]
   }
   n5 = "Constraint - deform_lowerarm_r (P)"(ComponentPose=@n4, Alpha=1, ConstraintWeights_0=1) {
    alphaBoolBlend {blendOption=Linear}
    boneToModify {boneName=deform_lowerarm_r}
    constraintSetup [{targetBone {boneName=lowerarm_r}}]
    constraintWeights [1]
   }
   n6 = "Constraint - deform_hand_r (P)"(ComponentPose=@n5, Alpha=1, ConstraintWeights_0=1) {
    alphaBoolBlend {blendOption=Linear}
    boneToModify {boneName=deform_hand_r}
    constraintSetup [{targetBone {boneName=deform_hand_r}}]
    constraintWeights [1]
   }
   n7 = Control Rig!(
    Source=Component To Local(ComponentPose=@n6)
    Alpha
    Crouching:double=Get Alpha_Crouching() -> (Alpha_Crouching:double)
    Emote:double=Get Alpha_Emote() -> (Alpha_Emote:double)
    Floating:double=Get Alpha_Floating() -> (Alpha_Floating:double)
    SelectedPistol:double=Get Selected_Pistol() -> (Selected_Pistol:double)
    Get_LargeGun_Equipped:double=Get large_gun_out() -> (large_gun_out:double)
    ShotGun_Out:double=Get PistolGunOut() -> (PistolGunOut:double)
    Get_Pistol_Out:double=Get ShortRifle Gun Out() -> (ShortRifle Gun Out:double)
    EmoteBool:bool=Get bIsPlayingEmote()
    isMannequin:bool="Get isMannequin "() -> ("isMannequin ":bool)
   ) -> (Pose) {
    alphaBoolBlend {blendOption=Linear}
    controlRigAssetReference {blueprintRigClass=&/M_MED_WartyBrine_Sus_CtrlRig.M_MED_WartyBrine_Sus_CtrlRig_C}
   }
   n8 = "Transform (Modify) Bone - Bone: chain_sd_l"(ComponentPose=Local To Component(LocalPose=@n7), Alpha=1, Translation="1.5,0,2", Rotation="0,-15,0", Scale="1,1,1") {
    rotationMode=BMM_Additive
    translationMode=BMM_Additive
    translationSpace=BCS_BoneSpace
    alphaBoolBlend {blendOption=Linear}
    boneToModify {boneName=chain_sd_l}
    rotation {yaw=-15}
    translation {x=1.5, z=2}
   }
   n9 = "Transform (Modify) Bone - Bone: chain_sd_l"(ComponentPose=@n8, Alpha=1, Translation="6.986,2.684,-4.116", Rotation="-13.902,-14.100,11.353", Scale="1,1,1") {
    rotationMode=BMM_Additive
    rotationSpace=BCS_BoneSpace
    scaleMode=BMM_Additive
    scaleSpace=BCS_BoneSpace
    translationMode=BMM_Additive
    translationSpace=BCS_BoneSpace
    alphaBoolBlend {blendOption=Linear}
    alphaScaleBiasClamp {bInterpResult=true}
    boneToModify {boneName=chain_sd_l}
    rotation=<rot -13.902 -14.1 11.353>
    translation=<6.986 2.684 -4.116>
   }
   n10 = "Transform (Modify) Bone - Bone: chain_bk_c"(ComponentPose=@n9, Alpha=1, Translation="-2.428,-2.508,-3.239", Rotation="0,0,0", Scale="1,1,1") {
    rotationMode=BMM_Additive
    rotationSpace=BCS_BoneSpace
    scaleMode=BMM_Additive
    scaleSpace=BCS_BoneSpace
    translationMode=BMM_Additive
    translationSpace=BCS_BoneSpace
    alphaBoolBlend {blendOption=Linear}
    alphaScaleBiasClamp {bInterpResult=true}
    boneToModify {boneName=chain_bk_c}
    translation=<-2.428 -2.508 -3.239>
   }
   n11 = "Constraint - surfaceChest (PPPPPP)"!(
    ComponentPose=@n10
    Alpha=1
    ConstraintWeights_0=0.413
    ConstraintWeights_1:float=0.2954
    ConstraintWeights_2:float=0.1418
    ConstraintWeights_3:float=0.133
    ConstraintWeights_4:float=0.0166
    ConstraintWeights_5:float=0.0002
   ) -> (Pose) {
    alphaBoolBlend {blendOption=Linear}
    boneToModify {boneName=surfaceChest}
    constraintSetup [
     {targetBone {boneName=spine_05}}
     {targetBone {boneName=spine_03}}
     {targetBone {boneName=spine_04}}
     {targetBone {boneName=spine_02}}
     {targetBone {boneName=spine_01}}
     {targetBone {boneName=pelvis}}
    ]
    constraintWeights [1,1,1,1,1,1]
   }
   n12 = "Transform (Modify) Bone - Bone: chain_sd_r"(ComponentPose=@n11, Alpha=1, Translation="-1.5,0,0", Rotation="0,0,0", Scale="1,1,1") {
    rotationMode=BMM_Additive
    rotationSpace=BCS_BoneSpace
    scaleMode=BMM_Additive
    scaleSpace=BCS_BoneSpace
    translationMode=BMM_Additive
    translationSpace=BCS_BoneSpace
    alphaBoolBlend {blendOption=Linear}
    alphaScaleBiasClamp {bInterpResult=true}
    boneToModify {boneName=chain_sd_r}
    translation {x=-1.5}
   }
   n13 = RigidBody!(ComponentPose=@n12, Alpha, OverrideWorldGravity:Vector="0,0,-980", ExternalForce) -> (Pose) {
    bOverrideWorldGravity=true
    overridePhysicsAsset=&/M_MED_WartyBrine_Sus_Physics
    simulationSpace=BaseBoneSpace
    alphaBoolBlend {blendOption=Linear}
    baseBoneRef {boneName=spine_02}
    componentAppliedLinearAccClamp=<0 0 0>
    simSpaceSettings {
     maxAngularAcceleration=1000
     maxAngularVelocity=3
     maxLinearAcceleration=700
     maxLinearVelocity=500
     velocityScaleZ=0
     worldAlpha=0.4
     externalLinearDragV=<0.2 0.2>
    }
   }
   Output Pose(Result=Component To Local(ComponentPose=@n13)) {layerGroup=None}
  ]}
 EventGraph {nodes [
   Event BlueprintUpdateAnimation()
   > n1 = Sequence!(execute) -> (then_0, then_1, then_2:Exec, then_3:Exec, then_4:Exec)
   n2 = Sequence!(execute=@n1.then_4) -> (then_0, then_1, then_2, then_3)
   n3 = GetAnimInstance(self=GetOwnerSkeletalMeshComponent())
   Is Valid(exec=@n2.then_2, InputObject=@n3)
   > n4 = Set IsBalletSpin(IsBalletSpin:bool=OR Boolean(
     A=Montage_IsPlaying(self=@n3, Montage=/BRCosmetics/Animation/Game/MainPlayer/Emotes/BalletSpin/Emote_BalletSpin_CMM_M)
     B=Montage_IsPlaying(self=@n3, Montage=/BRCosmetics/Animation/Game/MainPlayer/Emotes/BalletSpin/Emote_BalletSpin_CMF_M)
    )) -> (Output_Get:bool)
   > n5 = Branch(Condition=@n4)
   > Set IsBalletEmote(IsBalletEmote:double=1) -> (Output_Get:double)
   Set IsBalletEmote(execute=@n5.else, IsBalletEmote=0)
   Is Valid(exec=@n2.then_3, InputObject=@n3)
   > n6 = Set Iscelebration(Iscelebration:bool=OR Boolean(
     A=Montage_IsPlaying(self=@n3, Montage=/Game/Animation/Game/MainPlayer/Emotes/Emote_Celebration_M)
     B=Montage_IsPlaying(self=@n3, Montage=/BRCosmetics/Animation/Game/MainPlayer/Emotes/Mime/Emote_Mime_CMF_M)
    )) -> (Output_Get:bool)
   > n7 = Branch(Condition=@n6)
   > Set IsCelebrationEmote(IsCelebrationEmote:double=1) -> (Output_Get:double)
   Set IsCelebrationEmote(execute=@n7.else, IsCelebrationEmote=0)
   n8 = Cast To FortPlayerAnimInstance(execute=@n2.then_0, Object=GetAnimInstance(self=GetOwnerSkeletalMeshComponent()))
   > Set rifle_out(rifle_out:bool=false) -> (Output_Get:bool)
   > Set PistolOut(PistolOut:bool=false) -> (Output_Get:bool)
   > Set Short_Rifle_Out(Short_Rifle_Out:bool=false) -> (Output_Get:bool)
   > n9 = Switch on EFortWeaponCoreAnimation!(execute, Selection:EFortWeaponCoreAnimation Enum=Get WeaponCoreAnim(self=@n8)) -> (
    Melee:Exec
    Pistol:Exec
    Shotgun:Exec
    PaperBlueprint:Exec
    Rifle:Exec
    MeleeOneHand:Exec
    MachinePistol:Exec
    RocketLauncher:Exec
    GrenadeLauncher:Exec
    GoingCommando:Exec
    AssaultRifle:Exec
    TacticalShotgun:Exec
    SniperRifle:Exec
    TrapPlacement:Exec
    ShoulderLauncher:Exec
    AbilityDecoTool:Exec
    Crossbow:Exec
    C4:Exec
    RemoteControl:Exec
    DualWield:Exec
    AR_BullPup:Exec
    AR_ForwardGrip:Exec
    MedPackPaddles:Exec
    SMG_P90:Exec
    AR_DrumGun:Exec
    Consumable_Small:Exec
    Consumable_Large:Exec
    Balloon:Exec
    MountedTurret:Exec
    CreativeTool:Exec
    ExplosiveBow:Exec
    AshtonIndigo:Exec
    AshtonChicago:Exec
    MeleeDualWield:Exec
    Unarmed:Exec
   )
   Set rifle_out(execute=@n9.AssaultRifle, rifle_out=true)
   Set Short_Rifle_Out(execute=@n9.Pistol, Short_Rifle_Out=true)
   Set PistolOut(execute=@n9.Shotgun, PistolOut=true)
   Set large_gun_out(execute=@n2.then_1, large_gun_out:double="To Float (Boolean)"(InBool=Get rifle_out() -> (rifle_out:bool))) -> (Output_Get:double)
   > Set PistolGunOut(PistolGunOut:double="To Float (Boolean)"(InBool=Get PistolOut() -> (PistolOut:bool))) -> (Output_Get:double)
   > Set ShortRifle Gun Out(ShortRifle Gun Out:double="To Float (Boolean)"(InBool=Get Short_Rifle_Out() -> (Short_Rifle_Out:bool))) -> (Output_Get:double)
   Set Alpha_Crouching(execute=@n1.then_1, Alpha_Crouching:double="To Float (Boolean)"(InBool=Get bIsCrouching())) -> (Output_Get:double)
   > Set Alpha_Emote(Alpha_Emote:double="To Float (Boolean)"(InBool=Get bIsPlayingEmote())) -> (Output_Get:double)
   > Set Alpha_Floating(Alpha_Floating:double="To Float (Boolean)"(InBool=Get bIsSkydiveFloating())) -> (Output_Get:double)
   n10 = GetOwnerSkeletalMeshComponent()
   n11 = BreakRotator(InRot=BreakTransform(InTransform=GetDeltaTransformFromRefPose(self=@n10, BoneName=neck_02, BaseName=spine_05)).Rotation)
   n12 = Branch(execute=@n1.then_3, Condition=AND Boolean(A="float < float"(A=@n11.Yaw, B=-30), B=Get bIsCrouching()))
   > Set Alpha_HeadRotateUpCrouchExtrem(Alpha_HeadRotateUpCrouchExtrem:double=1) -> (Output_Get:double)
   Set Alpha_HeadRotateUpCrouchExtrem(execute=@n12.else, Alpha_HeadRotateUpCrouchExtrem=0)
   n13 = GetOwningActor()
   Cast To FortPlayerMannequin!(execute=@n1.then_0, Object=@n13) -> (then, CastFailed, AsFort Player Mannequin:Fort Player Mannequin Object Reference) {is=title K2Node_DynamicCast}
   > n14 = For Each Loop(Array:Array of Skeletal Mesh Component Object References=Get Components By Class!(self:Actor Object Reference=@n13, ComponentClass:Actor Component Class Reference=/Script/Engine.SkeletalMeshComponent) -> (ReturnValue:Array of Skeletal Mesh Component Object References)) -> (Array Element:Skeletal Mesh Component Object Reference)
   > n15 = Branch(Condition="Equal (Object)"(
     A=Get SkeletalMesh!(self:Skinned Mesh Component Object Reference=@n14) -> (SkeletalMesh:Skeletal Mesh Object Reference)
     B=&/M_MED_WartyBrine_Sus
    ))
   > "Set isMannequin "("isMannequin ":bool=false) -> (Output_Get:bool)
   "Set isMannequin "(execute=@n15.else, "isMannequin "=true)
   n16 = GetOwnerSkeletalMeshComponent()
   n17 = BreakRotator(InRot=BreakTransform(InTransform=GetDeltaTransformFromRefPose(self=@n16, BoneName=spine_05, BaseName=pelvis)).Rotation)
   n18 = Branch(execute=@n1.then_2, Condition=AND Boolean(A="float < float"(A=@n17, B=-12), B=Get bIsPlayingMeleeAnim()))
   > Set Alpha_ArmBackInMeleeL(Alpha_ArmBackInMeleeL:double=1) -> (Output_Get:double)
   Set Alpha_ArmBackInMeleeL(execute=@n18.else, Alpha_ArmBackInMeleeL=0)
   Event OnWeaponChanged()
   > n19 = Cast To FortPlayerAnimInstance(Object=GetAnimInstance(self=GetOwnerSkeletalMeshComponent()))
   > Set Selected_Pistol(Selected_Pistol:double=0) -> (Output_Get:double)
   > n20 = Switch on EFortWeaponCoreAnimation(Selection=Get WeaponCoreAnim(self=@n19))
   Set Selected_Pistol(execute=@n20.Pistol, Selected_Pistol=1)
  ]}
]
variables [
 Alpha_ArmBackInMeleeL {type=number}
 Alpha_Crouching {type=number}
 Alpha_Emote {type=number}
 Alpha_Floating {type=number}
 Alpha_HeadRotateUpCrouchExtrem {type=number}
 Alpha_Targeting {type=number}
 is sky diving {type=number}
 IsBalletEmote {type=number}
 IsBalletSpin {type=boolean}
 Iscelebration {type=boolean}
 IsCelebrationEmote {type=number}
 "isMannequin " {type=boolean}
 large_gun_out {type=number}
 PistolGunOut {type=number}
 PistolOut {type=boolean}
 rifle_out {type=boolean}
 Selected_Pistol {type=number}
 Short_Rifle_Out {type=boolean}
 ShortRifle Gun Out {type=number}
 Skel Mesh {type=SkeletalMeshComponent}
]
