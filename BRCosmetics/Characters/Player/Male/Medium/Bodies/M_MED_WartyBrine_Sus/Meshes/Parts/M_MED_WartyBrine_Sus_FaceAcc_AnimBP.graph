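# Post-process style AnimBP for the WartyBrine_Sus face accessory mesh, animating its hood or mask by cosmetic variant.
# On customization load, Stage1 and Stage2 meta tags set the hoodie and mask flags.
# A cached Copy Pose From Mesh feeds nested bool blends: mask AnimDynamics takes priority, then hoodie constraints, hood dynamics and corrective bone offsets, else the copied pose.
# Arm rotation alphas update each frame but go unused in the AnimGraph.
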AnimBlueprint /BRCosmetics/Characters/Player/Male/Medium/Bodies/M_MED_WartyBrine_Sus/Meshes/Parts/M_MED_WartyBrine_Sus_FaceAcc_AnimBP
target_skeleton=&/M_MED_WartyBrine_Sus_FaceAcc_Skeleton
graphs [
 AnimGraph {nodes [
   n1 = "Constraint - hoodie_side_r_base (PPPPP)"!(
    ComponentPose=Local To Component(LocalPose=Use cached pose 'StartCache'())
    Alpha=1
    ConstraintWeights_0=0.8975
    ConstraintWeights_1:float=0.0664
    ConstraintWeights_2:float=0.0209
    ConstraintWeights_3:float=0.0108
    ConstraintWeights_4:float=0.0044
   ) -> (Pose) {
    alphaBoolBlend {blendOption=Linear}
    boneToModify {boneName=hoodie_side_r_base}
    constraintSetup [
     {targetBone {boneName=neck_01}}
     {targetBone {boneName=spine_05}}
     {targetBone {boneName=clavicle_r}}
     {targetBone {boneName=spine_04}}
     {targetBone {boneName=spine_03}}
    ]
    constraintWeights [1,1,1,1,1]
   }
   n2 = "Constraint - hoodie_side_l_base (PPPPP)"!(
    ComponentPose=@n1
    Alpha=0.4
    ConstraintWeights_0=0.3708
    ConstraintWeights_1:float=0.3634
    ConstraintWeights_2:float=0.1822
    ConstraintWeights_3:float=0.0551
    ConstraintWeights_4:float=0.0285
   ) -> (Pose) {
    alphaBoolBlend {blendOption=Linear}
    boneToModify {boneName=hoodie_side_l_base}
    constraintSetup [
     {targetBone {boneName=neck_01}}
     {targetBone {boneName=spine_05}}
     {targetBone {boneName=clavicle_l}}
     {targetBone {boneName=spine_04}}
     {targetBone {boneName=spine_03}}
    ]
    constraintWeights [1,1,1,1,1]
   }
   n3 = Anim Dynamics - Bone: hood_top!(ComponentPose=@n2, Alpha=1, GravityOverride:Vector="0,0,-980", ExternalForce="0,0,0") -> (Pose) {
    angularDampingOverride=1
    angularSpringConstant=12000
    bAngularSpring=true
    bOverrideAngularDamping=true
    bUseGravityOverride=true
    gravityScale=0.35
    alphaBoolBlend {blendOption=Linear}
    boundBone {boneName=hood_top}
    componentAppliedLinearAccClamp=<3000 3000 3000>
    componentLinearAccScale=<0.3 0.3 0.3>
    componentLinearVelScale=<0.3 0.3 0.3>
    physicsBodyDefinitions [{
      boundBone {boneName=hood_top}
      boxExtents=<25 25 25>
      constraintSetup {angularLimitsMax=<0 40 0>, angularLimitsMin=<0 -20 0>, angularTarget=<1 0 0>}
      localJointOffset=<4 0 0>
     }]
   }
   n4 = "Transform (Modify) Bone - Bone: hoodie_side_l"(ComponentPose=@n3, Alpha=1, Translation="0,0,4.1", Rotation="0,0,-20", Scale="1,1,1") {
    rotationMode=BMM_Additive
    rotationSpace=BCS_BoneSpace
    scaleMode=BMM_Additive
    scaleSpace=BCS_BoneSpace
    translationMode=BMM_Additive
    alphaBoolBlend {blendOption=Linear}
    boneToModify {boneName=hoodie_side_l}
    rotation {roll=-20}
    translation {z=4.1}
   }
   n5 = "Transform (Modify) Bone - Bone: hoodie_fix_side_r"(ComponentPose=@n4, Alpha=1, Translation="-0.000,8.157,0.000", Rotation="0,0,0", Scale="1,1,1") {
    rotationMode=BMM_Additive
    rotationSpace=BCS_BoneSpace
    scaleMode=BMM_Additive
    scaleSpace=BCS_BoneSpace
    translationMode=BMM_Additive
    alphaBoolBlend {blendOption=Linear}
    boneToModify {boneName=hoodie_fix_side_r}
    translation {y=8.157}
   }
   n6 = "Transform (Modify) Bone - Bone: hoodie_side_l"(ComponentPose=@n5, Alpha=1, Translation="1.75,-0.75,0", Rotation="0,0,0", Scale="1,1,1") {
    rotationMode=BMM_Additive
    rotationSpace=BCS_BoneSpace
    scaleMode=BMM_Additive
    scaleSpace=BCS_BoneSpace
    translationMode=BMM_Additive
    alphaBoolBlend {blendOption=Linear}
    boneToModify {boneName=hoodie_side_l}
    translation=<1.75 -0.75>
   }
   n7 = "Transform (Modify) Bone - Bone: hoodie_side_r"(ComponentPose=@n6, Alpha=1, Translation="-5,0,0", Rotation="0,0,0", Scale="1,1,1") {translationMode=BMM_Additive, alphaBoolBlend {blendOption=Linear}, boneToModify {boneName=hoodie_side_r}, translation {x=-5}}
   n8 = Component To Local(ComponentPose=@n7)
   n9 = Use cached pose 'StartCache'()
   n10 = Get IsHoodyActive() -> (IsHoodyActive:bool)
   n11 = Anim Dynamics - Bone: mask_base_dyn!(
    ComponentPose=Local To Component(LocalPose=Use cached pose 'StartCache'())
    Alpha=1
    GravityOverride:Vector="0,0,-980"
    ExternalForce="0,0,0"
   ) -> (Pose) {
    angularDampingOverride=1
    angularSpringConstant=15000
    bAngularSpring=true
    bOverrideAngularDamping=true
    bUseGravityOverride=true
    gravityScale=0.35
    alphaBoolBlend {blendOption=Linear}
    boundBone {boneName=mask_base_dyn}
    componentAppliedLinearAccClamp=<3000 3000 3000>
    componentLinearAccScale=<0.2 0.2 0.2>
    componentLinearVelScale=<0.2 0.2 0.2>
    physicsBodyDefinitions [{
      boundBone {boneName=mask_base_dyn}
      boxExtents=<20 30 20>
      constraintSetup {angularLimitsMax=<0 35 5>, angularLimitsMin=<0 -30 -5>, angularTarget=<1 0 0>}
      localJointOffset=<3 0 0>
     }]
   }
   n12 = Anim Dynamics - Bone: mask_l_01_dyn!(ComponentPose=@n11, Alpha=1, GravityOverride:Vector="0,0,-980", ExternalForce="0,0,0") -> (Pose) {
    angularDampingOverride=1
    angularSpringConstant=13500
    bAngularSpring=true
    bOverrideAngularDamping=true
    bUseGravityOverride=true
    gravityScale=0.35
    alphaBoolBlend {blendOption=Linear}
    boundBone {boneName=mask_l_01_dyn}
    componentAppliedLinearAccClamp=<3000 3000 3000>
    componentLinearAccScale=<0.3 0.3 0.3>
    componentLinearVelScale=<0.3 0.3 0.3>
    physicsBodyDefinitions [{
      boundBone {boneName=mask_l_01_dyn}
      boxExtents=<25 25 25>
      constraintSetup {angularLimitsMax=<2 30 40>, angularLimitsMin=<-2 -25 -30>, angularTarget=<1 0 0>}
      localJointOffset=<5 0 0>
     }]
   }
   n13 = Anim Dynamics - Bone: mask_l_02_dyn!(ComponentPose=@n12, Alpha=1, GravityOverride:Vector="0,0,-980", ExternalForce="0,0,0") -> (Pose) {
    angularDampingOverride=1
    angularSpringConstant=10000
    bAngularSpring=true
    bOverrideAngularDamping=true
    bUseGravityOverride=true
    gravityScale=0.35
    alphaBoolBlend {blendOption=Linear}
    boundBone {boneName=mask_l_02_dyn}
    componentAppliedLinearAccClamp=<3000 3000 3000>
    componentLinearAccScale=<0.2 0.2 0.2>
    componentLinearVelScale=<0.2 0.2 0.2>
    physicsBodyDefinitions [{
      boundBone {boneName=mask_l_02_dyn}
      boxExtents=<20 20 20>
      constraintSetup {angularLimitsMax=<2 25 35>, angularLimitsMin=<-2 -20 -30>, angularTarget=<1 0 0>}
      localJointOffset=<4 0 0>
     }]
   }
   n14 = Anim Dynamics - Bone: mask_r_01_dyn!(ComponentPose=@n13, Alpha=1, GravityOverride:Vector="0,0,-980", ExternalForce="0,0,0") -> (Pose) {
    angularDampingOverride=1
    angularSpringConstant=13500
    bAngularSpring=true
    bOverrideAngularDamping=true
    bUseGravityOverride=true
    gravityScale=0.35
    alphaBoolBlend {blendOption=Linear}
    boundBone {boneName=mask_r_01_dyn}
    componentAppliedLinearAccClamp=<3000 3000 3000>
    componentLinearAccScale=<0.3 0.3 0.3>
    componentLinearVelScale=<0.3 0.3 0.3>
    physicsBodyDefinitions [{
      boundBone {boneName=mask_r_01_dyn}
      boxExtents=<25 25 25>
      constraintSetup {angularLimitsMax=<2 35 30>, angularLimitsMin=<-2 -40 -30>, angularTarget=<1 0 0>}
      localJointOffset=<5 0 0>
     }]
   }
   n15 = Anim Dynamics - Bone: mask_r_02_dyn!(ComponentPose=@n14, Alpha=1, GravityOverride:Vector="0,0,-980", ExternalForce="0,0,0") -> (Pose) {
    angularDampingOverride=1
    angularSpringConstant=10000
    bAngularSpring=true
    bOverrideAngularDamping=true
    bUseGravityOverride=true
    gravityScale=0.35
    alphaBoolBlend {blendOption=Linear}
    boundBone {boneName=mask_r_02_dyn}
    componentAppliedLinearAccClamp=<3000 3000 3000>
    componentLinearAccScale=<0.2 0.2 0.2>
    componentLinearVelScale=<0.2 0.2 0.2>
    physicsBodyDefinitions [{
      boundBone {boneName=mask_r_02_dyn}
      boxExtents=<20 20 20>
      constraintSetup {angularLimitsMax=<2 25 25>, angularLimitsMin=<-2 -25 -25>, angularTarget=<1 0 0>}
      localJointOffset=<4 0 0>
     }]
   }
   Output Pose(Result=Blend Poses by bool(
     BlendPose_0=Component To Local(ComponentPose=@n15)
     BlendPose_1=Blend Poses by bool(BlendPose_0=@n8, BlendPose_1=@n9, bActiveValue=@n10) {blendType=Linear}
     bActiveValue=Get IsMaskActive() -> (IsMaskActive:bool)
    ) {blendType=Linear}) {layerGroup=None}
   Save cached pose 'StartCache'(Pose=Copy Pose From Mesh() {bUseAttachedParent=true}) {cacheName=StartCache}
  ]}
 EventGraph {nodes [
   n1 = Event BlueprintUpdateAnimation()
   n2 = Event OnCharacterCustomizationLoaded()
   > n3 = Set Skel Mesh(Skel Mesh=@n2.CustomBaseMesh)
   Set DetectedArmLRot(
    execute=@n1.then
    DetectedArmLRot:double=GetAlphaFromReferenceTransformDelta(SkeletalMeshComponent=@n3, BoneName=upperarm_l, BaseName=clavicle_l, DeltaComponentType=EulerY, InputRange=(X=0,Y=60), TwistAxis="1,0,0")
   ) -> (Output_Get:double)
   > Set DetectedArmRRot(DetectedArmRRot:double=GetAlphaFromReferenceTransformDelta(SkeletalMeshComponent=@n3, BoneName=upperarm_r, BaseName=clavicle_r, DeltaComponentType=EulerY, InputRange=(X=0,Y=40), TwistAxis="1,0,0")) -> (Output_Get:double)
   Set IsHoodyActive(execute=@n7.then, IsHoodyActive:bool=true) -> (Output_Get:bool)
   Set IsHoodyActive(execute=@n7.else, IsHoodyActive=false)
   Set IsMaskActive(execute=@n8.else, IsMaskActive:bool=false) -> (Output_Get:bool)
   Set IsMaskActive(execute=@n8.then, IsMaskActive=true)
   n4 = CharacterItemDefinitionHasAnyMetaTag(Tags=Make Array("[0]"=(TagName="Cosmetics.Variant.Property.Stage1")))
   n5 = CharacterItemDefinitionHasAnyMetaTag(Tags=Make Array("[0]"=(TagName="Cosmetics.Variant.Property.Stage2")))
   n6 = Sequence(execute=@n3.then)
   > n7 = Branch(Condition=@n4)
   n8 = Branch(execute=@n6.then_1, Condition=@n5)
  ]}
]
variables [
 DetectedArmLRot {type=number}
 DetectedArmRRot {type=number}
 IsHoodyActive {type=boolean}
 IsMaskActive {type=boolean}
 Skel Mesh {type=SkeletalMeshComponent}
]
