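# Control Rig for the WartyBrine_Sus male character's face accessory part, correcting hoodie side bones as the body and arms move.
# Weighted parent constraints attach the left and right hoodie base bones to spine, neck and clavicle bones.
# Delta Fixes then offsets one corrective bone per side, weighted by upper-arm rotation relative to reference nulls.
# The reusable Delta Rotation function isolates swing or twist, picks an Euler axis and remaps it to a 0–1 weight.
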ControlRigBlueprint /BRCosmetics/Characters/Player/Male/Medium/Bodies/M_MED_WartyBrine_Sus/Meshes/Parts/M_MED_WartyBrine_Sus_FaceAcc_CtrlRig
graphs [
 New Function_ContainedGraph {nodes [
   n1 = RigVMFunctionEntryNode()
   RerouteNode_7_Constant = RigVMDispatchNode(Value=@n1.primAxis) {template="DISPATCH_RigVMDispatch_Constant(io Value)"}
   Clamp = &RigVMFunction_MathIntClamp(Value=@RerouteNode_7_Constant.Value)
   RigVMFunction_MathBoolToInteger = &RigVMFunction_MathBoolToInteger(Value=@n1.onlyTwist)
   RigVMFunction_MathQuaternionSwingTwist = &RigVMFunction_MathQuaternionSwingTwist(
    TwistAxis=@n1.twistAxis
    Input=&RigUnit_GetRelativeTransformForItem(
     Child=(Type=Bone,Name="None")
     Parent=(Type=Bone,Name="None")
     bChildInitial=False
     bParentInitial=False
     Parent.Type=@n1.ParentType
     Child.Name=@n1.child
     Parent.Name=@n1.parent
    ).RelativeTransform.Rotation
   )
   RerouteNode_Constant = RigVMDispatchNode(Value=@n1.rotationOrder) {template="DISPATCH_RigVMDispatch_Constant(io Value)"}
   RigVMFunction_MathQuaternionToEuler = &RigVMFunction_MathQuaternionToEuler(
    Value=RigVMDispatchNode(
     Values=((),())
     Index=@RigVMFunction_MathBoolToInteger.Result
     Values.0=@RigVMFunction_MathQuaternionSwingTwist.Swing
     Values.1=@RigVMFunction_MathQuaternionSwingTwist.Twist
    ) {template="DISPATCH_RigVMDispatch_SelectInt32(in Index,in Values,out Result)"}.Result
    RotationOrder=@RerouteNode_Constant.Value
   )
   DISPATCH_RigVMDispatch_SelectInt32 = RigVMDispatchNode(
    Values=(0,0,0)
    Index=@Clamp.Result
    Values.0=@RigVMFunction_MathQuaternionToEuler.Result.X
    Values.1=@RigVMFunction_MathQuaternionToEuler.Result.Y
    Values.2=@RigVMFunction_MathQuaternionToEuler.Result.Z
   ) {template="DISPATCH_RigVMDispatch_SelectInt32(in Index,in Values,out Result)"}
   RerouteNode_5_Constant = RigVMDispatchNode(Value=@n1.min) {template="DISPATCH_RigVMDispatch_Constant(io Value)"}
   RerouteNode_6_Constant = RigVMDispatchNode(Value=@n1.max) {template="DISPATCH_RigVMDispatch_Constant(io Value)"}
   Remap = &RigVMFunction_MathDoubleRemap(
    TargetMaximum=1
    TargetMinimum=0
    bClamp=False
    Value=@DISPATCH_RigVMDispatch_SelectInt32.Result
    SourceMinimum=@RerouteNode_5_Constant.Value
    SourceMaximum=@RerouteNode_6_Constant.Value
   )
   > RigVMFunctionReturnNode(ouputWeight=&RigVMFunction_MathDoubleClamp(Value=@Remap.Result).Result)
   RerouteNode_8_MakeStruct = RigVMDispatchNode(Elements=(X=1,Y=0,Z=0)) {template="DISPATCH_RigVMDispatch_MakeStruct(in Elements,out Struct)"}
  ]}
 New Function_ContainedGraph {nodes [
   RigVMFunctionEntryNode()
   > RigVMFunction_Sequence = RigVMAggregateNode()
   Delta Rotation_0 = RigVMFunctionReferenceNode(
    ParentType=Null
    child=upperarm_l
    debugRotation=false
    max=90
    min=0
    parent=upperarm_l_null
    primAxis=0
    rotationOrder=ZXY
    twistAxis=(X=0,Y=1,Z=0)
    ExecuteContext=@RigVMFunction_Sequence.A
   )
   > &RigUnit_ModifyTransforms(
    ItemToModify=((Item=(Type=Bone,Name="hoodie_side_l"),Transform=(Rotation=(X=0.1736,Y=0,Z=0,W=0.9848),Translation=(X=5.3,Y=0,Z=4.5),Scale3D=(X=1,Y=1,Z=1))))
    WeightMinimum=0
    Weight=@"Delta Rotation_0".ouputWeight
   )
   > Delta Rotation_1 = RigVMFunctionReferenceNode(
    ParentType=Null
    child=upperarm_r
    debugRotation=false
    max=56
    min=0
    parent=upperarm_r_null
    primAxis=0
    rotationOrder=ZXY
    twistAxis=(X=0,Y=1,Z=0)
   )
   > &RigUnit_ModifyTransforms(
    ItemToModify=((Item=(Type=Bone,Name="hoodie_fix_side_r"),Transform=(Rotation=(X=0,Y=0,Z=0,W=1),Translation=(X=0,Y=26,Z=0),Scale3D=(X=1,Y=1,Z=1))))
    WeightMinimum=0
    Weight=@"Delta Rotation_1".ouputWeight
   )
   RigVMFunctionReturnNode(ExecuteContext=@RigVMFunction_Sequence.B)
  ]}
 RigVMFunctionLibrary {nodes [
   Delta Fixes = RigVMCollapseNode(ExecuteContext=())
   Delta Rotation = RigVMCollapseNode(ExecuteContext=(), twistAxis=(X=0,Y=0,Z=0))
  ]}
 RigVMFunction_Sequence_1_ContainedGraph {nodes [
   n1 = RigVMFunctionEntryNode()
   RigVMFunction_Sequence = &RigVMFunction_Sequence(ExecuteContext=@n1.ExecuteContext)
   n2 = &RigVMFunction_Sequence(ExecuteContext=@RigVMFunction_Sequence.B)
   RigVMFunctionReturnNode(A=@RigVMFunction_Sequence.A, B=@n2.A, C=@n2.B)
  ]}
 RigVMModel {nodes [
   RigUnit_BeginExecution = &RigUnit_BeginExecution()
   &RigUnit_ParentConstraint(
    AdvancedSettings=(RotationOrderForFilter=XZY)
    Child=(Type=Bone,Name="hoodie_side_l_base")
    Parents=((Item=(Type=Bone,Name="spine_03"),Weight=0.0285),(Item=(Type=Bone,Name="spine_04"),Weight=0.0551),(Item=(Type=Bone,Name="clavicle_l"),Weight=0.1822),(Item=(Type=Bone,Name="spine_05"),Weight=0.3634),(Item=(Type=Bone,Name="neck_01"),Weight=0.3708))
    Weight=1
    ExecutePin=@RigUnit_BeginExecution.ExecutePin
   )
   > &RigUnit_ParentConstraint(
    AdvancedSettings=(RotationOrderForFilter=XZY)
    Child=(Type=Bone,Name="hoodie_side_r_base")
    Parents=((Item=(Type=Bone,Name="spine_03"),Weight=0.0044),(Item=(Type=Bone,Name="spine_04"),Weight=0.0108),(Item=(Type=Bone,Name="clavicle_r"),Weight=0.0209),(Item=(Type=Bone,Name="spine_05"),Weight=0.0664),(Item=(Type=Bone,Name="neck_01"),Weight=0.8975))
    Weight=1
   )
   > Delta Fixes = RigVMFunctionReferenceNode()
  ]}
]
hierarchy [
 root {loc=<0 0 0>, rot=<0 0 0 1>}
 pelvis {parent=root, loc=<0 -0.7071 93.1191>, rot=<0.0127 0.707 -0.0127 -0.707>}
 spine_01 {parent=pelvis, loc=<6.7494 -0.6388 0>, rot=<0 0 0.034 -0.9994>}
 spine_02 {parent=spine_01, loc=<7.252 -0.035 0>, rot=<0 0 0.0238 -0.9997>}
 spine_03 {parent=spine_02, loc=<6.0106 0 0>, rot=<0 0 0.0149 -0.9999>}
 spine_04 {parent=spine_03, loc=<8.1023 0 0>, rot=<0 0 -0.0132 -0.9999>}
 spine_05 {parent=spine_04, loc=<10.0142 0 0>, rot=<0 0 -0.0015 -1>}
 neck_01 {parent=spine_05, loc=<11.3138 0.1529 0>, rot=<0 0 -0.2003 -0.9797>}
 neck_02 {parent=neck_01, loc=<4.5627 0 0>, rot=<0 0 0.0504 -0.9987>}
 head {parent=neck_02, loc=<5.3914 0 0>, rot=<-0.0415 -0.0131 0.1181 -0.992>}
 faceAttach {parent=head, loc=<-4.146 -0.8545 0>, rot=<0.0029 -0.7071 -0.0029 -0.7071>}
 mask_base_dyn {parent=neck_01, loc=<15.5922 -19.2866 -0.7148>, rot=<0.4131 -0.5739 0.5739 -0.4131>}
 mask_l_01_dyn {parent=mask_base_dyn, loc=<3.3727 5.1329 -2.6674>, rot=<0.9068 0.204 -0.3594 -0.0826>}
 mask_l_02_dyn {parent=mask_l_01_dyn, loc=<5.9088 0 0>, rot=<-0.0071 0.1207 -0.0189 -0.9925>}
 mask_r_01_dyn {parent=mask_base_dyn, loc=<3.6987 -5.8657 1.0011>, rot=<0.9593 -0.2391 -0.1415 0.0496>}
 mask_r_02_dyn {parent=mask_r_01_dyn, loc=<6.717 0 0>, rot=<-0.0009 0.1137 0.0082 -0.9935>}
 hood_top {parent=neck_01, loc=<28.2447 -14.5112 -6.1919>, rot=<0.1455 -0.6021 0.3016 -0.7248>}
 hoodie_side_l_base {parent=neck_01, loc=<7.0777 5.5595 -17.764>, rot=<-0.0542 -0.6347 0.0755 -0.7671>}
 hoodie_side_l {parent=hoodie_side_l_base, loc=<0 0 0>, rot=<0 0 0 1>}
 hoodie_side_r_base {parent=neck_01, loc=<8.692 5.6327 16.8302>, rot=<0.7629 0.0334 0.6197 0.1815>}
 hoodie_side_r {parent=hoodie_side_r_base, loc=<0 0 0>, rot=<0 0 0 1>}
 hoodie_fix_side_r {parent=hoodie_side_r_base, loc=<0 0 0>, rot=<0 0 0 1>}
 clavicle_l {parent=spine_05, loc=<4.3575 -0.1466 -2.3223>, rot=<0.0465 -0.7289 -0.009 -0.6829>}
 upperarm_l {parent=clavicle_l, loc=<15.5111 0 0>, rot=<0.0226 -0.3234 -0.0088 -0.9459>}
 upperarm_l_null {parent=clavicle_l, type=Null, loc=<0 0 0>, rot=<0 0 0 1>}
 clavicle_r {parent=spine_05, loc=<4.3575 -0.1466 2.3223>, rot=<0.7289 0.0465 -0.6829 0.009>}
 upperarm_r {parent=clavicle_r, loc=<-15.5111 0 0>, rot=<0.0226 -0.3234 -0.0088 -0.9459>}
 upperarm_r_null {parent=clavicle_r, type=Null, loc=<4.7494 5.3064 7.969>, rot=<0.0482 -0.06 -0.435 0.8971>}
]
